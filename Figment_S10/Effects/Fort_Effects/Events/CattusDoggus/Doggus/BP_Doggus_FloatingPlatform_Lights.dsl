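; Doggus event floating platform lights Blueprint that turns a particle system light effect on and off.
; Custom ActivateLights and DeactivateLights events activate or deactivate the particle system component, with activation not forcing a reset.
; A Debug flag also activates the effect on BeginPlay.
; Tick, overlap and construction script are empty.
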(event Custom|ActivateLights
  (Components|Activation|Activate (Variables|Default|GetParticleSystemComponent0) false))

(event EventBeginPlay
  (if (Variables|Default|GetDebug)
    (Components|Activation|Activate (Variables|Default|GetParticleSystemComponent0) false)))

(event Collision|EventActorBeginOverlap
  (params (OtherActor :type Actor%20Object%20Reference)))

(event EventTick
  (params (DeltaSeconds :type Float%20%28single-precision%29)))

(event Custom|DeactivateLights
  (Components|Activation|Deactivate (Variables|Default|GetParticleSystemComponent0)))

(fn ConstructionScript)
