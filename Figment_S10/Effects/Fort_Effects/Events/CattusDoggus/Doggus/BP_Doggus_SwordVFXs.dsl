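; Blueprint for the Doggus sword VFX in the CattusDoggus event effects.
; Its construction script sets the "Random" scalar material parameter on the Sword_ZeroPtHair component to a 0–1 float drawn from a random stream variable, varying that material per placed instance.
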(fn ConstructionScript
  (Rendering|Material|SetScalarParameterValueonMaterials (Variables|Default|GetSword_ZeroPtHair) "Random" (Math|Random|RandomFloatinRangefromStream (Variables|Default|GetRandomStream) 0.0 1.0)))
